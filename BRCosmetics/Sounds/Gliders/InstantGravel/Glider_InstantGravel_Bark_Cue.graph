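# Sound cue for the InstantGravel glider's bark, a cosmetic one-shot that borrows the InstantGravel back bling's react recordings.
# A Random node picks one of seven equally weighted Backpack_InstantGravel_React waves.
# The cue plays at reduced volume in the Backblings sound class, despite being a glider sound, and is spatialized with the close-detail player glider attenuation.
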SoundCue /BRCosmetics/Sounds/Gliders/InstantGravel/Glider_InstantGravel_Bark_Cue
root=Random_0
nodes [
 {
  id=Random_0
  children [WavePlayer_0,WavePlayer_1,WavePlayer_2,WavePlayer_3,WavePlayer_4,WavePlayer_5,WavePlayer_6]
  properties {weights [1,1,1,1,1,1,1]}
 }
 {id=WavePlayer_0, properties {soundWaveAssetPtr=/BRCosmetics/Sounds/Backblings/InstantGravel/Backpack_InstantGravel_React_01}}
 {id=WavePlayer_1, properties {soundWaveAssetPtr=/BRCosmetics/Sounds/Backblings/InstantGravel/Backpack_InstantGravel_React_02}}
 {id=WavePlayer_2, properties {soundWaveAssetPtr=/BRCosmetics/Sounds/Backblings/InstantGravel/Backpack_InstantGravel_React_03}}
 {id=WavePlayer_3, properties {soundWaveAssetPtr=/BRCosmetics/Sounds/Backblings/InstantGravel/Backpack_InstantGravel_React_04}}
 {id=WavePlayer_4, properties {soundWaveAssetPtr=/BRCosmetics/Sounds/Backblings/InstantGravel/Backpack_InstantGravel_React_05}}
 {id=WavePlayer_5, properties {soundWaveAssetPtr=/BRCosmetics/Sounds/Backblings/InstantGravel/Backpack_InstantGravel_React_06}}
 {id=WavePlayer_6, properties {soundWaveAssetPtr=/BRCosmetics/Sounds/Backblings/InstantGravel/Backpack_InstantGravel_React_07}}
]
properties {
 attenuationSettings=/Game/Sounds/Attenuation/Fort_Player_Glider_CloseDetail_Athena_Att
 maxDistance=6000
 soundClassObject=/Game/Sounds/FortSoundClassesAndModes/Fort_Backblings
 volumeMultiplier=0.4
}
